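# Physics asset for the Rumble Female two-handed pickaxe, giving the handle collision shapes and its decorative ribbon a jointed body chain.
# Two capsules and a box cover the handle, while capsules cover the ribbon root, knot, two bows and two three-segment tails.
# Constraints branch from handle through root and knot to bows and tails, with tight twist limits along the tails.
# The root-to-handle and left-master-to-knot joints rotate freely, unlike the tightly limited right master.
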PhysicsAsset /BRCosmetics/Weapons/FORT_Melee/Pickaxe_Rumble_Female/Meshes/Rumble_Female_Axe_Physics
bodies {
 Melee_TwoHanded_Handle [
  {
   shapeType=Capsule
   length=66.5682
   radius=3.8435
   shapeName=Melee_TwoHanded_Handle_capsule
   center=<0.0489 -1.0328 46.4515>
   rotation=<rot -0.0087 -180 -0.1027>
  }
  {shapeType=Capsule, length=24, radius=6, shapeName=Melee_TwoHanded_Handle_capsule, center=<0.3661 -0.9151 60.617>}
  {shapeType=Box, extentX=1, extentY=8.5, extentZ=7.5, shapeName=Melee_TwoHanded_Handle_box, center=<1.6713 -1.2891 86.7529>}
 ]
 dyn_ribbon_0_LFT [{
   shapeType=Capsule
   length=3.005
   radius=0.755
   shapeName=dyn_ribbon_0_LFT_capsule
   center=<2.0883 0 -0.0621>
   rotation=<rot 0 -90.0004 89.9998>
  }]
 dyn_ribbon_0_RGT [{
   shapeType=Capsule
   length=3.005
   radius=0.755
   shapeName=dyn_ribbon_0_RGT_capsule
   center=<2.1669 0.0212 -0.0583>
   rotation=<rot -20.0002 -89.9995 90.0001>
  }]
 dyn_ribbon_1_LFT [{shapeType=Capsule, length=5.255, radius=1.255, shapeName=dyn_ribbon_1_LFT_capsule, center=<3.3047 0 0>, rotation=<rot 0 89.9997 90.0002>}]
 dyn_ribbon_1_RGT [{
   shapeType=Capsule
   length=5.8909
   radius=1.2545
   shapeName=dyn_ribbon_1_RGT_capsule
   center=<3.3817 0 0>
   rotation=<rot 10.0001 90.0005 90.1984>
  }]
 dyn_ribbon_2_LFT [{
   shapeType=Capsule
   length=5.8342
   radius=1.3582
   shapeName=dyn_ribbon_2_LFT_capsule
   center=<3.9593 0 0>
   rotation=<rot 9.9999 89.9997 -89.9997>
  }]
 dyn_ribbon_2_RGT [{
   shapeType=Capsule
   length=6.0715
   radius=1.4244
   shapeName=dyn_ribbon_2_RGT_capsule
   center=<3.8907 0 0>
   rotation=<rot -20 90.0004 -90.0001>
  }]
 dyn_ribbon_LFT_MSTR [{shapeType=Capsule, length=0.505, radius=0.505, shapeName=dyn_ribbon_LFT_MSTR_capsule}]
 dyn_ribbon_RGT_MSTR [{shapeType=Capsule, length=0.505, radius=0.505, shapeName=dyn_ribbon_RGT_MSTR_capsule}]
 dyn_ribbon_bow_LFT [{
   shapeType=Capsule
   length=1.3215
   radius=1.4244
   shapeName=dyn_ribbon_2_RGT_capsule
   center=<1.8237 0 -0.6602>
   rotation=<rot 0.0001 90.0003 -90>
  }]
 dyn_ribbon_bow_RGT [{
   shapeType=Capsule
   length=1.5715
   radius=1.4244
   shapeName=dyn_ribbon_2_RGT_capsule
   center=<1.6217 0.008 -0.5244>
   rotation=<rot 10.0001 90.0006 -90.0001>
  }]
 dyn_ribbon_knot [{shapeType=Capsule, length=0.7315, radius=1.2213, shapeName=dyn_ribbon_knot_capsule}]
 dyn_ribbon_root [{shapeType=Capsule, length=0.505, radius=0.505, shapeName=dyn_ribbon_root_capsule}]
}
constraints [
 {
  bone1Name=dyn_ribbon_root
  bone2Name=Melee_TwoHanded_Handle
  swing1LimitDegrees=45
  swing1Motion=Free
  swing2LimitDegrees=45
  swing2Motion=Free
  twistLimitDegrees=45
  twistMotion=Free
 }
 {bone1Name=dyn_ribbon_knot, bone2Name=dyn_ribbon_root, swing1LimitDegrees=5, swing2LimitDegrees=10, twistLimitDegrees=5}
 {bone1Name=dyn_ribbon_RGT_MSTR, bone2Name=dyn_ribbon_knot, swing1LimitDegrees=5, swing2LimitDegrees=5, twistLimitDegrees=2}
 {bone1Name=dyn_ribbon_0_RGT, bone2Name=dyn_ribbon_RGT_MSTR, swing1LimitDegrees=45, swing2LimitDegrees=45, twistLimitDegrees=5}
 {bone1Name=dyn_ribbon_1_RGT, bone2Name=dyn_ribbon_0_RGT, swing1LimitDegrees=45, swing2LimitDegrees=45, twistLimitDegrees=5}
 {bone1Name=dyn_ribbon_2_RGT, bone2Name=dyn_ribbon_1_RGT, swing1LimitDegrees=45, swing2LimitDegrees=45, twistLimitDegrees=5}
 {
  bone1Name=dyn_ribbon_LFT_MSTR
  bone2Name=dyn_ribbon_knot
  swing1LimitDegrees=45
  swing1Motion=Free
  swing2LimitDegrees=45
  swing2Motion=Free
  twistLimitDegrees=45
  twistMotion=Free
 }
 {bone1Name=dyn_ribbon_0_LFT, bone2Name=dyn_ribbon_LFT_MSTR, swing1LimitDegrees=45, swing2LimitDegrees=45, twistLimitDegrees=5}
 {bone1Name=dyn_ribbon_1_LFT, bone2Name=dyn_ribbon_0_LFT, swing1LimitDegrees=45, swing2LimitDegrees=45, twistLimitDegrees=5}
 {bone1Name=dyn_ribbon_2_LFT, bone2Name=dyn_ribbon_1_LFT, swing1LimitDegrees=45, swing2LimitDegrees=45, twistLimitDegrees=5}
 {bone1Name=dyn_ribbon_bow_LFT, bone2Name=dyn_ribbon_knot, swing1LimitDegrees=30, swing2LimitDegrees=30, twistLimitDegrees=2}
 {bone1Name=dyn_ribbon_bow_RGT, bone2Name=dyn_ribbon_knot, swing1LimitDegrees=30, swing2LimitDegrees=30, twistLimitDegrees=2}
]
